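; Base photo-mode camera mode that applies tag-selected facing data through a camera modifier while it is active.
; On activation it reads facing data from the view target's photo-mode pawn anim component.
; It re-enables an already tracked modifier found on the player camera manager, or adds a new one and passes it the facing values.
; On deactivation it disables the modifier non-immediately and clears the stored reference.
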(event EventOnBecomeActiveBP
  (params (ViewTarget :type Actor%20Object%20Reference))
  (bind _facingdata (FacingData|GetFacingData (Actor|GetComponentbyClass ViewTarget "/PhotoMode/Blueprints/B_PM_PlayerPawnAnim_Component.B_PM_PlayerPawnAnim_Component_C") (Variables|Default|GetFacing%20Tag)))
  (Variables|Default|SetFacing%20Data _facingdata)
  ; Update camera pitch
  (Variables|Default|GetFacingCameraModifier
    (:then
      (bind _returnvalue (Game|Player|FindCameraModifierbyClass (CallFunction|GetPlayerCameraManager) (Variables|Default|GetCameraModifierClass)))
      (Utilities|IsValid _returnvalue
        (:Is%20Valid
          (CameraModifier|EnableModifier _returnvalue))
        (:Is%20Not%20Valid
          _)))
    (:else
      _))
  (bind _returnvalue_2 (Game|Player|AddNewCameraModifier (CallFunction|GetPlayerCameraManager) (Variables|Default|GetCameraModifierClass)))
  (bind _output_get (Variables|Default|SetFacingCameraModifier _returnvalue_2))
  (Class|CMPMBaseFacing|UpdateFacingValues _output_get (Variables|Default|GetFacing%20Data)))

(event EventOnBecomeInactiveBP
  (params (ViewTarget :type Actor%20Object%20Reference))
  (bind _facingcameramodifier (Variables|Default|GetFacingCameraModifier)
    (:then
      (CameraModifier|DisableModifier _facingcameramodifier false)
      (Variables|Default|SetFacingCameraModifier 0))
    (:else)))
